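; GA_Quest_VisitHardwood is a quest gameplay ability that tracks the player visiting a featured island, apparently the Hardwood one given its name.
; On activation it runs the parent logic, then waits for the featured island gameplay tag to be added.
; When the tag appears, it sends a complex custom stat event for the quest tag to the owning player's quest manager, then ends itself.
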(event Ability|EventActivateAbility
  (|CallParentFunction)
  (Ability|Tasks|WaitGameplayTagAdd (Variables|Default|GetT_FeaturedIsland)
    (:Added
      (Quest|SendComplexCustomStatEvent (Variables|GAVars|Data|GetOwningPlayerQuestManager) 0 (Variables|Default|GetT_Quest) "(GameplayTags=)" 1)
      (Ability|EndAbility))))

(event Ability|EventFailedToActivatePassiveAbility
  (params (ActorInfo :type Gameplay%20Ability%20Actor%20Info%20Structure)))

(event Ability|EventOnEndAbility
  (params (bWasCancelled :type Boolean)))
